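# Temp looping Sound Cue for an MMO capture effect in BasilOkra, on the Fort gadgets sound class.
# A looping wave feeds a Modulator that fixes its pitch lower, then an Enveloper whose curve climbs to a higher pitch over time while volume stays flat.
# It references a quest-loot attenuation asset.
# The cue's volume multiplier is zero, so it is silent as saved, and its debug flag is on.
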SoundCue /WildEstate_BasilOkra/Assets/Temp/SFX/MMO/Cue_BasilOkra_MMO_Capture_Loop
root=Enveloper_1
nodes [
 {
  id=Enveloper_1
  children [Modulator_0]
  properties {pitchCurve {editorCurveData {keys [<at 0 0.95>, {time=0}, <at 50 1.5>]}}, volumeCurve {editorCurveData {keys [<at 0 1>,<at 1 1>]}}}
 }
 {id=Modulator_0, children [WavePlayer_0], properties {pitchMax=0.7, pitchMin=0.7, volumeMax=1, volumeMin=1}}
 {id=WavePlayer_0, properties {bLooping=true, soundWaveAssetPtr=&/Train_MMO_Capture_Loop_BasilOkra}}
]
properties {
 attenuationSettings=&/Train_MMOQuest_Loot_Att_BasilOkra
 bDebug=true
 maxDistance=5600
 soundClassObject=&Fort_Gadgets
 volumeMultiplier=0
 attenuationOverrides {bSpatialize=false}
 concurrencyOverrides {bLimitToOwner=true, maxCount=3, resolutionRule=StopOldest}
}
