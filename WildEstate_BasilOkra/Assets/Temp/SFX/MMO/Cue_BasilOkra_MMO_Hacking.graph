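# Temporary MMO sound cue for hacking, named for BasilOkra, that plays the hack start sound in the gadgets sound class.
# A single wave player feeds a Modulator node left at its default settings.
# Attenuation comes from a shared MMO quest loot asset.
# The cue also stores non-spatialized attenuation and per-owner, stop-oldest concurrency override values, though no enabling override flags are shown, and has its debug flag on.
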SoundCue /WildEstate_BasilOkra/Assets/Temp/SFX/MMO/Cue_BasilOkra_MMO_Hacking
root=Modulator_0
nodes [
 {id=Modulator_0, children [WavePlayer_1]}
 {id=WavePlayer_1, properties {soundWaveAssetPtr=&/Train_MMO_Hack_Start_BasilOkra}}
]
properties {
 attenuationSettings=&/Train_MMOQuest_Loot_Att_BasilOkra
 bDebug=true
 maxDistance=5600
 soundClassObject=&Fort_Gadgets
 volumeMultiplier=0.9
 attenuationOverrides {bSpatialize=false}
 concurrencyOverrides {bLimitToOwner=true, maxCount=3, resolutionRule=StopOldest}
}
